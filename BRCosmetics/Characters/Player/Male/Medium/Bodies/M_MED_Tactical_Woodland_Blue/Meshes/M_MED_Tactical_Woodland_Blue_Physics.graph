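# Physics asset for the M_MED_Tactical_Woodland_Blue male body cosmetic, covering its two dynamic pouch bones.
# Box bodies on dyn_hip_pouch and dyn_left_pouch are constrained to small placeholder capsules on spine_01 and thigh_twist_01_l respectively.
# The hip pouch has room to swing, widest on its second swing axis, while the left thigh pouch is held almost rigid by very tight limits.
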PhysicsAsset /BRCosmetics/Characters/Player/Male/Medium/Bodies/M_MED_Tactical_Woodland_Blue/Meshes/M_MED_Tactical_Woodland_Blue_Physics
bodies {
 dyn_hip_pouch [{shapeType=Box, extentX=4, extentY=10.7448, extentZ=16.7374, center=<3.0536 -1.7377 0.0853>, rotation=<rot 0 85.6619 -84.8822>}]
 dyn_left_pouch [{shapeType=Box, extentX=4, extentY=11.5131, extentZ=18.2841, center=<4.2169 -2.3821 0.4949>, rotation=<rot 0 88.0988 -88.6143>}]
 spine_01 [{shapeType=Capsule, length=1, radius=1}]
 thigh_twist_01_l [{shapeType=Capsule, length=1, radius=1}]
}
constraints [
 {bone1Name=dyn_hip_pouch, bone2Name=spine_01, swing1LimitDegrees=10, swing2LimitDegrees=45, twistLimitDegrees=10}
 {bone1Name=dyn_left_pouch, bone2Name=thigh_twist_01_l, swing1LimitDegrees=1, swing2LimitDegrees=1, twistLimitDegrees=1}
]
